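# Unlit translucent material that displays the T_GolfCourse biome mask as a tinted overlay projected top-down onto the scene behind the mesh.
# UVs come from world position behind translucency, relative to actor position, scaled by object bounds and centered, with depth test disabled.
# The larger of red and blue blends blue to magenta; green is unused.
# That value, maxed with alpha and scaled by an Opacity parameter, forms the transparency.
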Material /BRCoreMaterials/Hera/Terrain/BiomeMask/T_GolfCourse_Mat
graph {
 properties {
  bAutomaticallySetUsageInEditor=false
  bCastRayTracedShadows=false
  bDisableDepthTest=true
  blendMode=BLEND_Translucent
  shadingModel=MSM_Unlit
  usedShadingModels=Unlit
 }
 nodes [
  n1 = Subtract(
   A=ComponentMask(None=MaterialFunctionCall() {materialFunction=&WorldPositionBehindTranslucency}.Result) {g=true, r=true}
   B=ComponentMask(None=ActorPositionWS()) {g=true, r=true}
  )
  n2 = ComponentMask(None=Multiply(A=ObjectBounds()) {constB=2}) {g=true, r=true}
  n3 = TextureSample(UVs=Subtract(A=Divide(A=@n1, B=@n2)) {constB=-0.5}) {samplerType=SAMPLERTYPE_LinearColor, texture=&/T_GolfCourse}
  n4 = Max(A=Max(A=@n3.R, B=@n3.R), B=@n3.B)
  LinearInterpolate(A=Constant3Vector() {constant=<rgba 0 0.5187 1 1>}, B=Constant3Vector() {constant=<rgba 1 0 0.9911 1>}, Alpha=@n4)
  Multiply(A=Max(A=@n4, B=@n3.A), B=ScalarParameter() {defaultValue=0.5, parameterName=Opacity})
 ]
}
