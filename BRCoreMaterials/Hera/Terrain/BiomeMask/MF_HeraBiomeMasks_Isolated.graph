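# Material function for Hera terrain that splits four packed RGBA mask inputs (QuadA to QuadD) into named per-biome mask outputs.
# Most channels pass straight through, while BiomeSettings scalar parameters can force Chaparral and Archipelago on.
# Desert is masked out by NarrowFlea, GoldenFall2 squares the GoldenFall mask, and the landscape grass removal mask is inverted.
# A world-position-projected golf course texture branch is present but feeds none of the outputs.
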MaterialFunction /BRCoreMaterials/Hera/Terrain/BiomeMask/MF_HeraBiomeMasks_Isolated
graph {nodes [
  n1 = FunctionInput() {bUsePreviewValueAsDefault=true, inputName=TerrainMasks_QuadA, inputType=FunctionInput_Vector4, sortPriority=-500, previewValue=<0 0 0 0>}
  n2 = MaterialFunctionCall(Float4=NamedRerouteDeclaration(None=@n1) {name=TerrainMasks_QuadA, nodeColor=<rgba 0.5 0 1 1>}) {materialFunction=&BreakOutFloat4Components}
  n3 = ScalarParameter() {group=BiomeSettings, parameterName=ForceChaparral}
  FunctionOutput(None=Saturate(None=Add(A=@n2.R, B=@n3))) {outputName=00_Chaparral, sortPriority=-1000}
  FunctionOutput(None=@n2.G) {outputName=01_RedWoods, sortPriority=-990}
  n4 = FunctionInput() {bUsePreviewValueAsDefault=true, inputName=TerrainMasks_QuadC, inputType=FunctionInput_Vector4, sortPriority=-300, previewValue=<0 0 0 0>}
  n5 = MaterialFunctionCall(Float4=NamedRerouteDeclaration(None=@n4) {name=TerrainMasks_QuadC, nodeColor=<rgba 0.5 0 1 1>}) {materialFunction=&BreakOutFloat4Components}
  FunctionOutput(None=@n5.B) {outputName=10_IslandC, sortPriority=-900}
  FunctionOutput(None=@n5.A) {outputName=11_RemoveStripes, sortPriority=-890}
  n6 = FunctionInput() {bUsePreviewValueAsDefault=true, inputName=TerrainMasks_QuadD, inputType=FunctionInput_Vector4, sortPriority=-200, previewValue=<0 0 0 0>}
  n7 = MaterialFunctionCall(Float4=NamedRerouteDeclaration(None=@n6) {name=TerrainMasks_QuadD, nodeColor=<rgba 0.5 0 1 1>}) {materialFunction=&BreakOutFloat4Components}
  FunctionOutput(None=@n7.R) {outputName=12_NarrowFleaVariation, sortPriority=-880}
  FunctionOutput(None=@n7.G) {outputName=13_None, sortPriority=-870}
  FunctionOutput(None=@n7.B) {outputName=14_None, sortPriority=-860}
  FunctionOutput(None=OneMinus(None=@n7.A)) {outputName=15_RemoveLandscapeGrassTypes, sortPriority=-850}
  n8 = FunctionInput() {bUsePreviewValueAsDefault=true, inputName=TerrainMasks_QuadB, inputType=FunctionInput_Vector4, sortPriority=-400, previewValue=<0 0 0 0>}
  n9 = MaterialFunctionCall(Float4=NamedRerouteDeclaration(None=@n8) {name=TerrainMasks_QuadB, nodeColor=<rgba 0.5 0 1 1>}) {materialFunction=&BreakOutFloat4Components}
  FunctionOutput(None=@n9.A) {outputName=07_AxisMandate, sortPriority=-915}
  FunctionOutput(None=@n5.R) {outputName=08_IslandA, sortPriority=-912}
  n10 = ScalarParameter() {group=BiomeSettings, parameterName=ForceArchipelago}
  FunctionOutput(None=Saturate(None=Add(A=@n2.B, B=@n10))) {outputName=02_Archipelago, sortPriority=-980}
  n11 = Max(A=@n9.B, B=@n9.B)
  FunctionOutput(None=Multiply(A=@n2.A, B=OneMinus(None=@n11))) {outputName=03_Desert, sortPriority=-970}
  FunctionOutput(None=@n9.R) {outputName=04_GoldenFall, sortPriority=-960}
  FunctionOutput(None=Multiply(A=@n9.R, B=@n9.R)) {outputName=05_GoldenFall2, sortPriority=-950}
  FunctionOutput(None=@n9.B) {outputName=06_NarrowFlea, sortPriority=-940}
  FunctionOutput(None=@n5.G) {outputName=09_IslandB, sortPriority=-910}
  n12 = Subtract(A=WorldPosition().XY, B=Constant2Vector() {g=-5580.9795, r=83773.3594})
  n13 = Constant2Vector() {g=60000, r=30000}
  n14 = TextureSample(UVs=Subtract(A=Divide(A=@n12, B=@n13)) {constB=-0.5}) {
   samplerSource=SSM_Clamp_WorldGroupSettings
   samplerType=SAMPLERTYPE_LinearColor
   texture=&/T_GolfCourse
  }
  Multiply(A=Constant() {desc=07_A_GolfCourse_Rough}, B=@n14.R)
  Multiply(A=Constant() {desc=07_B_GolfCourse_Fairway}, B=@n14.G)
  Multiply(A=Constant() {desc=07_C_GolfCourse_PuttingGreen})
  NamedRerouteDeclaration(None=OneMinus(None=@n14.A)) {name=GolfCourse_Erase, nodeColor=<rgba 0 1 0.8594 1>}
 ]}
